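# Swing sound cue for the RedRiding pickaxe cosmetic.
# A mixer layers two attenuated branches: close uses RedRiding-specific swing waves, medium uses shared Fire Axe distant swings.
# Each branch runs through a modulator and a quality-level switch, which picks among six random variations, three, or a single wave.
# Overridden concurrency caps it at three instances, stopping the quietest, and it primes on load.
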SoundCue /BRCosmetics/Sounds/PickAxes/RedRiding/PA_RedRiding_Swing
root=Mixer_0
nodes [
 {id=Attenuation_0, children [Modulator_1], properties {attenuationSettings=/Game/Sounds/Attenuation/Fort_Weapon_Pickaxe_Close_Att}}
 {id=Attenuation_1, children [Modulator_0], properties {attenuationSettings=/Game/Sounds/Attenuation/Fort_Weapon_Pickaxe_Medium_Att}}
 {id=Mixer_0, children [Attenuation_0, Attenuation_1], properties {inputVolume [1,1]}}
 {id=Modulator_0, children [QualityLevel_1]}
 {id=Modulator_1, children [QualityLevel_2]}
 {id=QualityLevel_1, children [Random_1, Random_2, WavePlayer_15, WavePlayer_15]}
 {id=QualityLevel_2, children [Random_5, Random_6, WavePlayer_35, WavePlayer_35]}
 {
  id=Random_1
  children [WavePlayer_0, WavePlayer_6, WavePlayer_7, WavePlayer_8, WavePlayer_9, WavePlayer_11]
  properties {weights [1,1,1,1,1,1]}
 }
 {id=Random_2, children [WavePlayer_13, WavePlayer_12, WavePlayer_14], properties {weights [1,1,1]}}
 {
  id=Random_5
  children [WavePlayer_26, WavePlayer_27, WavePlayer_28, WavePlayer_29, WavePlayer_30, WavePlayer_31]
  properties {weights [1,1,1,1,1,1]}
 }
 {id=Random_6, children [WavePlayer_32, WavePlayer_33, WavePlayer_34], properties {weights [1,1,1]}}
 {id=WavePlayer_0, properties {soundWaveAssetPtr=/Game/Athena/Sounds/Weapons/Fire_Axe_Swing_Distant_04}}
 {id=WavePlayer_11, properties {soundWaveAssetPtr=/Game/Athena/Sounds/Weapons/Fire_Axe_Swing_Distant_06}}
 {id=WavePlayer_12, properties {soundWaveAssetPtr=/Game/Athena/Sounds/Weapons/Fire_Axe_Swing_Distant_03}}
 {id=WavePlayer_13, properties {soundWaveAssetPtr=&Fire_Axe_Swing_Distant_04}}
 {id=WavePlayer_14, properties {soundWaveAssetPtr=&Fire_Axe_Swing_Distant_06}}
 {id=WavePlayer_15, properties {soundWaveAssetPtr=&Fire_Axe_Swing_Distant_06}}
 {id=WavePlayer_26, properties {soundWaveAssetPtr=&/PA_RedRiding_Swing_01}}
 {id=WavePlayer_27, properties {soundWaveAssetPtr=&/PA_RedRiding_Swing_02}}
 {id=WavePlayer_28, properties {soundWaveAssetPtr=&/PA_RedRiding_Swing_03}}
 {id=WavePlayer_29, properties {soundWaveAssetPtr=&/PA_RedRiding_Swing_04}}
 {id=WavePlayer_30, properties {soundWaveAssetPtr=&/PA_RedRiding_Swing_05}}
 {id=WavePlayer_31, properties {soundWaveAssetPtr=&/PA_RedRiding_Swing_06}}
 {id=WavePlayer_32, properties {soundWaveAssetPtr=&PA_RedRiding_Swing_01}}
 {id=WavePlayer_33, properties {soundWaveAssetPtr=&PA_RedRiding_Swing_02}}
 {id=WavePlayer_34, properties {soundWaveAssetPtr=&PA_RedRiding_Swing_03}}
 {id=WavePlayer_35, properties {soundWaveAssetPtr=&PA_RedRiding_Swing_01}}
 {id=WavePlayer_6, properties {soundWaveAssetPtr=/Game/Athena/Sounds/Weapons/Fire_Axe_Swing_Distant_01}}
 {id=WavePlayer_7, properties {soundWaveAssetPtr=/Game/Athena/Sounds/Weapons/Fire_Axe_Swing_Distant_02}}
 {id=WavePlayer_8, properties {soundWaveAssetPtr=&Fire_Axe_Swing_Distant_03}}
 {id=WavePlayer_9, properties {soundWaveAssetPtr=/Game/Athena/Sounds/Weapons/Fire_Axe_Swing_Distant_05}}
]
properties {
 audioPropertiesSheet=/BRCosmetics/Sounds/Settings/PropertySheets/Pickaxe/APS_Cosmetic_Pickaxe_Swing
 bOverrideConcurrency=true
 bPrimeOnLoad=true
 maxDistance=3000
 soundClassObject=&Fort_Impacts
 volumeMultiplier=0.3
 attenuationOverrides {bAttenuateWithLPF=true, bEnableListenerFocus=true, distanceAlgorithm=Logarithmic, falloffDistance=800, attenuationShapeExtents {x=300}}
 concurrencyOverrides {maxCount=3, resolutionRule=StopQuietest}
}
